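# Sound cue for the end-of-swing sound of the LeafyVest cosmetic pickaxe.
# It randomly picks one of four equally weighted swing-end waves, varies pitch and volume slightly through a modulator, then applies the shared close-range pickaxe attenuation.
# It uses the melee weapon sound class and pickaxe swing property sheet, with a concurrency override that caps voices and blocks new plays at the limit.
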SoundCue /BRCosmetics/Sounds/PickAxes/LeafyVest/PickaxeSwingEnd_LeafyVest_Cue
root=Attenuation_0
nodes [
 {id=Attenuation_0, children [Modulator_0], properties {attenuationSettings=/Game/Sounds/Attenuation/Fort_Weapon_Pickaxe_Close_Att}}
 {id=Modulator_0, children [Random_0], properties {pitchMax=1.02, pitchMin=0.98, volumeMax=1.02, volumeMin=0.98}}
 {id=Random_0, children [WavePlayer_0,WavePlayer_1,WavePlayer_2,WavePlayer_3], properties {weights [1,1,1,1]}}
 {id=WavePlayer_0, properties {soundWaveAssetPtr=&/Pickaxe_LeafyVest_SwingEnd_01}}
 {id=WavePlayer_1, properties {soundWaveAssetPtr=&/Pickaxe_LeafyVest_SwingEnd_02}}
 {id=WavePlayer_2, properties {soundWaveAssetPtr=&/Pickaxe_LeafyVest_SwingEnd_03}}
 {id=WavePlayer_3, properties {soundWaveAssetPtr=&/Pickaxe_LeafyVest_SwingEnd_04}}
]
properties {
 audioPropertiesSheet=/BRCosmetics/Sounds/Settings/PropertySheets/Pickaxe/APS_Cosmetic_Pickaxe_Swing
 bOverrideConcurrency=true
 maxDistance=1536
 pitchMultiplier=0.99
 soundClassObject=&Fort_WeaponMelee_PSM
 volumeMultiplier=0.25
 concurrencyOverrides {maxCount=4, resolutionRule=PreventNew, voiceStealReleaseTime=0.5, volumeScale=0.5, volumeScaleAttackTime=0.25}
 concurrencySet [&Player_Melee_Weapon_Conc]
}
